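# Animation Blueprint for the ShiverFlame backpack that adds secondary motion to its lace and star trinket.
# With anim dynamics enabled, chained Anim Dynamics nodes simulate these bones, colliding with spine and pack limits; otherwise a trail controller and fixed bone offsets apply.
# The event graph copies body-mesh spine transforms onto the pack skeleton for collision, updates gravity override and zeroes component velocity and acceleration scales while skydiving.
# Pack and lace root offsets finish the pose.
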AnimBlueprint /BRCosmetics/Accessories/FORT_Backpacks/Backpack_F_MED_ShiverFlame/Meshes/F_MED_ShiverFlame_Pack_AnimBP
target_skeleton=&/F_MED_ShiverFlame_Pack_Skeleton
graphs [
 AnimGraph {nodes [
   n1 = Trail controller - Bone: dyn_trinket_ring!(ComponentPose=Local To Component(LocalPose=Use cached pose 'SavedPose9688'()), bAlphaBoolEnabled:bool=True) -> (Pose) {
    alphaInputType=Bool
    bLimitRotation=true
    bUsePlanarLimit=true
    alphaBoolBlend {blendInTime=1, blendOption=ExpIn, blendOutTime=0.5}
    alphaScaleBiasClamp {scale=0.7, inRange {min=0.5}}
    baseJoint {boneName=pack}
    planarLimits [
     {drivingBone {boneName=dyn_lace_MSTR}, planeTransform {location=<2 0 0>, rotation=<rot -90 -56.262 0>, scale=<1 1 1>}}
     {drivingBone {boneName=pack}, planeTransform {location=<0 -6 0>, rotation=<rot 0 0 -90>, scale=<1 1 1>}}
    ]
    rotationLimits [{limitMax=<90 90 90>, limitMin=<-90 -90 -90>}, {limitMax=<0 90 90>, limitMin=<-90 -90 -90>}]
    rotationOffsets [<0 0 0>, <0 0 0>]
    trailBone {boneName=dyn_trinket_ring}
    trailRelaxationSpeed {editorCurveData {keys [<at 0 10 RCIM_Linear>,<at 4.243 21.1174 RCIM_Linear>,<at 5.4765 50 RCIM_Linear>]}}
   }
   n2 = "Transform (Modify) Bone - Bone: dyn_lace_02"!(ComponentPose=@n1, bAlphaBoolEnabled:bool=True, Translation="0,0,0", Rotation="0,-90,0", Scale="1,1,1") -> (Pose) {
    alphaInputType=Bool
    rotationMode=BMM_Additive
    rotationSpace=BCS_BoneSpace
    translationMode=BMM_Additive
    translationSpace=BCS_BoneSpace
    alphaBoolBlend {blendInTime=0.5, blendOption=QuinticInOut, blendOutTime=1}
    boneToModify {boneName=dyn_lace_02}
    rotation {yaw=-90}
   }
   n3 = "Transform (Modify) Bone - Bone: dyn_trinket_star"(ComponentPose=@n2, Alpha=1, Translation="0,0,0", Rotation="-7.310,-28.323,-9.208", Scale="1,1,1") {
    rotationMode=BMM_Additive
    rotationSpace=BCS_BoneSpace
    translationMode=BMM_Additive
    translationSpace=BCS_BoneSpace
    alphaBoolBlend {blendOption=Linear}
    boneToModify {boneName=dyn_trinket_star}
    rotation=<rot -7.31 -28.323 -9.208>
   }
   n4 = "Transform (Modify) Bone - Bone: dyn_lace_02"!(ComponentPose=@n3, Alpha:float=1, Translation="0,0,0", Rotation="10,0,0", Scale="1,1,1") -> (Pose) {
    rotationMode=BMM_Additive
    rotationSpace=BCS_BoneSpace
    translationMode=BMM_Additive
    translationSpace=BCS_BoneSpace
    alphaBoolBlend {blendOption=Linear}
    boneToModify {boneName=dyn_lace_02}
    rotation {pitch=10}
   }
   n5 = Component To Local(ComponentPose=@n4)
   n6 = "Transform (Modify) Bone - Bone: spine_02"!(
    ComponentPose=Local To Component(LocalPose=Use cached pose 'SavedPose9688'())
    bAlphaBoolEnabled:bool=True
    Translation=Get Spine_02_Translation() -> (Spine_02_Translation:Vector)
    Rotation=Get Spine_02_Rotation() -> (Spine_02_Rotation:Rotator)
    Scale="-0.329,-0.329,1.000"
   ) -> (Pose) {
    alphaInputType=Bool
    rotationMode=BMM_Replace
    rotationSpace=BCS_WorldSpace
    scaleMode=BMM_Additive
    translationMode=BMM_Replace
    translationSpace=BCS_WorldSpace
    alphaBoolBlend {blendInTime=1, blendOption=Linear, blendOutTime=1}
    boneToModify {boneName=spine_02}
    rotation {roll=20}
    scale=<-0.329 -0.329>
   }
   n7 = "Transform (Modify) Bone - Bone: spine_03"!(
    ComponentPose=@n6
    bAlphaBoolEnabled:bool=True
    Translation=Get Spine_03_Translation() -> (Spine_03_Translation:Vector)
    Rotation=Get Spine_03_Rotation() -> (Spine_03_Rotation:Rotator)
    Scale="1,1,1"
   ) -> (Pose) {
    alphaInputType=Bool
    rotationMode=BMM_Replace
    rotationSpace=BCS_WorldSpace
    scaleMode=BMM_Additive
    translationMode=BMM_Replace
    translationSpace=BCS_WorldSpace
    alphaBoolBlend {blendInTime=1, blendOption=Linear, blendOutTime=1}
    boneToModify {boneName=spine_03}
    rotation {roll=20}
   }
   n8 = Break Fort Anim Input Skydiving External Force(FortAnimInput_SkydivingExternalForce=Get SkydivingExternalForceInput()) -> (
    bUseSkydivingVectorForce=~
    SkydivingVectorForce=~
    HeadToPelvisDirection=~
    FloatingMultiplier=~
    FloatingAdditive=~
    DivingMultiplier=~
    DivingAdditive=~
    ParachutingMultiplier=~
    ParachutingAdditive=~
   )
   n9 = Get Component Linear Acc Scale() -> (Component Linear Acc Scale:Vector)
   n10 = Get Component Linear Vel Scale() -> (Component Linear Vel Scale:Vector)
   n11 = Get Gravity Override()
   n12 = Anim Dynamics - Bone: dyn_lace!(
    ComponentPose=@n7
    Alpha=1
    GravityOverride:Vector=@n11
    ComponentLinearAccScale:Vector=@n9
    ComponentLinearVelScale:Vector=@n10
    ExternalForce=@n8.SkydivingVectorForce
   ) -> (Pose) {
    angularBiasOverride=1
    angularDampingOverride=1
    angularSpringConstant=3000
    bAngularSpring=true
    bOverrideAngularBias=true
    bOverrideAngularDamping=true
    bUseGravityOverride=true
    bUseSphericalLimits=true
    simulationSpace=RootRelative
    alphaBoolBlend {blendOption=Linear}
    boundBone {boneName=dyn_lace}
    componentAppliedLinearAccClamp=<3000 3000 3000>
    componentLinearAccScale=<0.4 0.5 0.5>
    componentLinearVelScale=<0.5 0.5 0.5>
    physicsBodyDefinitions [{
      collisionType=CustomSphere
      sphereCollisionRadius=2.2
      boundBone {boneName=dyn_lace}
      boxExtents=<12 12 12>
      constraintSetup {angularLimitsMax=<0 60 10>, angularLimitsMin=<0 -60 -55>, angularTarget=<0 0 0>}
      localJointOffset=<5 0 0>
     }]
    planarLimits [{drivingBone {boneName=pack}, planeTransform {location=<0.354 -0.2986 0>, rotation=<rot 0.0001 10 -90.0003>}}]
    sphericalLimits [
     {limitRadius=14, drivingBone {boneName=spine_02}, sphereLocalOffset=<0 0 0>}
     {limitRadius=14, drivingBone {boneName=spine_03}, sphereLocalOffset=<7.6385 0 0>}
     {limitRadius=13, drivingBone {boneName=pack}, sphereLocalOffset=<-18.4199 10.4376 -24.6877>}
     {limitRadius=15, drivingBone {boneName=pack}, sphereLocalOffset=<0 13.796 -16.2864>}
    ]
   }
   n13 = Anim Dynamics - Bone: dyn_lace_01!(
    ComponentPose=@n12
    Alpha=1
    GravityOverride:Vector=@n11
    ComponentLinearAccScale:Vector=@n9
    ComponentLinearVelScale:Vector=@n10
    ExternalForce=@n8.SkydivingVectorForce
   ) -> (Pose) {
    angularBiasOverride=1
    angularDampingOverride=1
    angularSpringConstant=3000
    bAngularSpring=true
    bOverrideAngularBias=true
    bOverrideAngularDamping=true
    bUseGravityOverride=true
    bUseSphericalLimits=true
    simulationSpace=RootRelative
    alphaBoolBlend {blendOption=Linear}
    boundBone {boneName=dyn_lace_01}
    componentAppliedLinearAccClamp=<3000 3000 3000>
    componentLinearAccScale=<0.4 0.5 0.5>
    componentLinearVelScale=<0.5 0.5 0.5>
    physicsBodyDefinitions [{
      collisionType=CustomSphere
      sphereCollisionRadius=3
      boundBone {boneName=dyn_lace_01}
      boxExtents=<10 10 10>
      constraintSetup {angularLimitsMax=<0 50 45>, angularLimitsMin=<0 -50 -45>, angularTarget=<0 0 0>}
      localJointOffset=<6 0 0>
     }]
    planarLimits [{drivingBone {boneName=pack}, planeTransform {location=<0.152 0.847 0>, rotation=<rot 0.0001 10 -90.0003>}}]
    sphericalLimits [
     {limitRadius=14, drivingBone {boneName=spine_02}, sphereLocalOffset=<0 1.1731 -0.0132>}
     {limitRadius=14, drivingBone {boneName=spine_03}, sphereLocalOffset=<7.6385 0 0>}
     {limitRadius=12.5, drivingBone {boneName=pack}, sphereLocalOffset=<-18.4199 9.1785 -24.6877>}
    ]
   }
   n14 = Anim Dynamics - Bone: dyn_lace_02!(
    ComponentPose=@n13
    Alpha=1
    GravityOverride:Vector=@n11
    ComponentLinearAccScale:Vector=@n9
    ComponentLinearVelScale:Vector=@n10
    ExternalForce=@n8.SkydivingVectorForce
   ) -> (Pose) {
    angularBiasOverride=1
    angularDampingOverride=1
    angularSpringConstant=175
    bAngularSpring=true
    bOverrideAngularBias=true
    bOverrideAngularDamping=true
    bUseGravityOverride=true
    bUsePlanarLimit=false
    simulationSpace=RootRelative
    alphaBoolBlend {blendOption=Linear}
    boundBone {boneName=dyn_lace_02}
    componentAppliedLinearAccClamp=<1500 1500 1200>
    componentLinearAccScale=<0.5 0.5 0.5>
    componentLinearVelScale=<0.6 0.6 0.6>
    physicsBodyDefinitions [{
      collisionType=CustomSphere
      sphereCollisionRadius=3
      boundBone {boneName=dyn_lace_02}
      boxExtents=<8 8 8>
      constraintSetup {angularLimitsMax=<0 45 20>, angularLimitsMin=<0 -45 -20>, angularTarget=<0 0 0>}
      localJointOffset=<2 0 0>
     }]
   }
   n15 = Anim Dynamics - Bone: dyn_trinket_ring!(ComponentPose=@n14, Alpha=1, GravityOverride:Vector=@n11, ExternalForce=@n8.SkydivingVectorForce) -> (Pose) {
    angularBiasOverride=1
    angularDampingOverride=1
    angularSpringConstant=350
    bAngularSpring=true
    bOverrideAngularBias=true
    bOverrideAngularDamping=true
    bUseGravityOverride=true
    bUsePlanarLimit=false
    bUseSphericalLimits=true
    simulationSpace=BoneRelative
    alphaBoolBlend {blendOption=Linear}
    boundBone {boneName=dyn_trinket_ring}
    componentAppliedLinearAccClamp=<2000 2000 2000>
    componentLinearAccScale=<0.3 0.3 0.4>
    componentLinearVelScale=<0.3 0.3 0.4>
    gravityOverride {z=-980}
    physicsBodyDefinitions [{
      boundBone {boneName=dyn_trinket_ring}
      boxExtents=<6.5 6.5 7>
      constraintSetup {angularLimitsMax=<15 80 30>, angularLimitsMin=<-15 -80 -65>, angularTarget=<0 0 0>}
      localJointOffset=<1 0 0>
     }]
    relativeSpaceBone {boneName=pack}
    sphericalLimits [
     {limitRadius=16.2322, drivingBone {boneName=spine_02}, sphereLocalOffset=<0 0 0>}
     {limitRadius=3.302, drivingBone {boneName=pack}, sphereLocalOffset=<-6.7043 -5.5591 -5.2002>}
    ]
   }
   n16 = Anim Dynamics - Bone: dyn_trinket_star!(ComponentPose=@n15, Alpha=1, GravityOverride:Vector=@n11, ExternalForce=@n8.SkydivingVectorForce) -> (Pose) {
    angularBiasOverride=1
    angularDampingOverride=1
    angularSpringConstant=150
    bAngularSpring=true
    bOverrideAngularBias=true
    bOverrideAngularDamping=true
    bUseGravityOverride=true
    bUsePlanarLimit=false
    simulationSpace=BoneRelative
    alphaBoolBlend {blendOption=Linear}
    boundBone {boneName=dyn_trinket_star}
    componentAppliedLinearAccClamp=<2000 2000 2000>
    componentLinearAccScale=<1 1 0.5>
    componentLinearVelScale=<0.1 1 0.5>
    gravityOverride {z=-980}
    physicsBodyDefinitions [{
      boundBone {boneName=dyn_trinket_star}
      boxExtents=<10 10 9>
      constraintSetup {angularLimitsMax=<0 30 -5>, angularLimitsMin=<0 -10 -30>, angularTarget=<0 0 0>}
      localJointOffset=<2 0 0>
     }]
    relativeSpaceBone {boneName=pack}
   }
   n17 = Component To Local(ComponentPose=@n16)
   n18 = Get bAnimDynamicsEnabled() -> (bAnimDynamicsEnabled:bool)
   n19 = "Transform (Modify) Bone - Bone: pack"!(
    ComponentPose=Local To Component(LocalPose=Blend Poses by bool(BlendPose_0=@n17, BlendPose_1=@n5, BlendTime_0=0.35, BlendTime_1=0.35, bActiveValue=@n18) {blendType=Linear})
    bAlphaBoolEnabled:bool
    Translation="0,0,0"
    Rotation="0,0,-7"
    Scale
   ) -> (Pose) {
    alphaInputType=Bool
    rotationMode=BMM_Additive
    rotationSpace=BCS_BoneSpace
    translationMode=BMM_Additive
    translationSpace=BCS_BoneSpace
    alphaBoolBlend {blendInTime=0.3, blendOption=ExpIn, blendOutTime=0.3}
    alphaScaleBiasClamp {bInterpResult=true, interpSpeedDecreasing=5}
    boneToModify {boneName=pack}
    rotation {roll=-7}
   }
   n20 = "Transform (Modify) Bone - Bone: dyn_lace_MSTR"!(ComponentPose=@n19, bAlphaBoolEnabled:bool=True, Translation="0,0,0", Rotation="0,-10,0", Scale="1,1,1") -> (Pose) {
    alphaInputType=Bool
    rotationMode=BMM_Additive
    rotationSpace=BCS_BoneSpace
    translationMode=BMM_Additive
    translationSpace=BCS_BoneSpace
    alphaBoolBlend {blendInTime=0.3, blendOption=ExpIn, blendOutTime=0.3}
    alphaScaleBiasClamp {bInterpResult=true, interpSpeedDecreasing=5}
    boneToModify {boneName=dyn_lace_MSTR}
    rotation {yaw=-10}
   }
   Output Pose(Result=Component To Local(ComponentPose=@n20)) {layerGroup=None}
   Save cached pose 'SavedPose9688'(Pose=Input Pose()) {cacheName=SavedPose9688}
  ]}
 EventGraph {nodes [
   Event BlueprintUpdateAnimation()
   > Is Valid(InputObject=Get BodyMesh() -> (BodyMesh:Skeletal Mesh Component Object Reference))
   n1 = Get BodyMesh()
   n2 = BreakTransform(InTransform=GetSocketTransform(self=@n1, InSocketName=spine_02, TransformSpace=RTS_World))
   > Set Spine_02_Translation(Spine_02_Translation:Vector=@n2) -> (Output_Get:Vector)
   > Set Spine_02_Rotation(Spine_02_Rotation:Rotator=@n2.Rotation) -> (Output_Get:Rotator)
   n3 = BreakTransform(InTransform=GetSocketTransform(self=@n1, InSocketName=spine_03, TransformSpace=RTS_World))
   > Set Spine_03_Translation(Spine_03_Translation:Vector=@n3) -> (Output_Get:Vector)
   > Set Spine_03_Rotation(Spine_03_Rotation:Rotator=@n3.Rotation) -> (Output_Get:Rotator)
   n4 = OR Boolean(A=Get bIsJumping(), B=Get bIsFalling())
   > Set notJumFall(notJumFall:bool=NOT Boolean(A=@n4)) -> (Output_Get:bool)
   > Set JumFall(JumFall:bool=@n4) -> (Output_Get:bool)
   > n5 = Gravity Override State Function DEPRECATED(CharacterPartAnimInstance=Self-Reference(), GameTimeInSeconds=GetGameTimeInSeconds(), GravityOverrideParams=Get Gravity Override Params())
   > Set Gravity Override(Gravity Override=@n5)
   > n6 = Branch(Condition=Get bIsSkyDiving())
   > Set Component Linear Acc Scale(Component Linear Acc Scale:Vector="0, 0, 0") -> (Output_Get:Vector)
   > Set Component Linear Vel Scale(Component Linear Vel Scale:Vector="0, 0, 0") -> (Output_Get:Vector)
   Set Component Linear Acc Scale(execute=@n6.else, Component Linear Acc Scale="0.5,0.5,0.5")
   > Set Component Linear Vel Scale(Component Linear Vel Scale="0.5,0.5,0.5")
   Event OnCharacterCustomizationLoaded()
   > n7 = Cast To PlayerPawn_Athena(Object=TryGetPawnOwner())
   > Set PlayerPawnAthena(PlayerPawnAthena:Player Pawn Athena Object Reference=@n7) -> (Output_Get:Player Pawn Athena Object Reference)
   > Set BodyMesh(BodyMesh=GetSkeletalMeshForPartType(self=@n7))
   TryGetPawnOwner()
   TryGetPawnOwner()
   Get bIsCrouching()
  ]}
]
variables [
 BodyMesh {type=SkeletalMeshComponent}
 Clock On {type=boolean}
 Component Linear Acc Scale {type=Vector}
 Component Linear Vel Scale {type=Vector}
 Gravity Override {type=Vector}
 Gravity Override Params {type=GravityOverrideParamsStruct}
 JumFall {type=boolean}
 Key_backward_fix {type=number}
 notJumFall {type=boolean}
 PlayerPawnAthena {type=PlayerPawn_Athena_C}
 Spine_02_Rotation {type=Rotator}
 Spine_02_Translation {type=Vector}
 Spine_03_Rotation {type=Rotator}
 Spine_03_Translation {type=Vector}
]
